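# Niagara effect for the jellyfish glider chute that ejects a burst of small bubble sprites.
# The Bubbles emitter spawns one of two random counts on a cylinder surface, nudges them outward and downward, then lifts them with upward acceleration against drag.
# Cascade-converted orbit adds wobble while curves swell, shrink and fade each bubble.
# A disabled Empty emitter is the only reader of User.IsAlt, passing it to a dynamic material parameter.
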NiagaraSystem /BRCosmetics/Gadgets/Assets/VinderTech_GliderChute/Glider_Jellyfish/FX/NS_JellyfishGlider_Bubble_Ejection
dependencies {
 usedDataInterfaces [&NiagaraDataInterfaceCurve]
 usedDynamicInputs []
 usedModules [
  &self:SystemSpawnScript.SystemScriptSource.SystemScriptGraph.NiagaraNodeAssignment_0.SetVariables_#1
  &SystemState
 ]
}
emitters {
 Bubbles {
  emitterUpdateScript [
   &EmitterState(Life Cycle Mode=::NewEnumerator1, Loop Behavior=::NewEnumerator1, Loop Delay=0.01, Loop Duration=1, UseLoopDelay=true)
   &SpawnBurst_Instantaneous(
    Spawn Count=&MakeCustomIntFromBool(Bool=&RandomBool(), True INT=16, False INT=32)
    Use Spawn Probability=false
    UseLoopCountLimit=false
   )
   &SpawnBurst_Instantaneous(Spawn Count=1000, Use Spawn Probability=false, UseLoopCountLimit=false) {enabled=false, module=SpawnBurst_Instantaneous001}
  ]
  particleSpawnScript [
   &InitializeParticle(
    Color=&LerpLinearColors(
     StartColor=<rgba 1 1 1 1>
     EndColor=&LerpLinearColors(
      StartColor=<rgba 0 1 1 1>
      EndColor=<rgba 1 0 1 1>
      LerpFactor=&UniformAOrBFloat(A=0, Evaluation Type=::NewEnumerator0, Recalculate Random Each Loop=true, Override Seed=false)
     )
     LerpFactor=0
    )
    Lifetime=&RandomRangeFloat(Minimum=0.5, Maximum=2.5)
    Sprite Rotation Angle=&RandomRangeFloat(Minimum=0, Maximum=1)
    Sprite Rotation Mode=::NewEnumerator1
    Sprite Size=&Vector2DFromFloat(Value=&RandomRangeFloat(Minimum=3, Maximum=5))
    Sprite Size Mode=::NewEnumerator4
    UsePositionOffset=false
    Write Material Random=true
   )
   &ShapeLocation(
    Cylinder Height=50
    Cylinder Height Midpoint=1
    Cylinder Radius=70
    MaskSpawnGroup=false
    Override Seed=false
    Rotation Axis=<0 0 1>
    Rotation Mode=::NewEnumerator1
    Shape Primitive=::NewEnumerator1
    Surface Only=true
    UseOwnerScale=true
   )
   &AddVelocity(Velocity=&Multiply_VectorByFloat(
     Vector=&Multiply_Vector(A=$Particles.ShapeLocation.ShapeVector, B=<1 1 0>)
     Float=&RandomRangeFloat(Minimum=0.9, Maximum=1.2)
    ))
   &Color(Color=&MakeLinearColorFromVectorAndFloat(
     "Vector (RGB)"=&RandomRangeVector(Minimum=<1.5 1.5 1.5>, Maximum=<1 1 1>)
     "Float (Alpha)"=&ClampFloat(Float=1)
    )) {enabled=false}
   &AddVelocity(Velocity=&RandomRangeVector(Minimum=<0 0 -100>, Maximum=<0 0 -10>)) {module=AddVelocity001}
   &InheritVelocity(Inherited Velocity Amount Scale=<-100 -100 1>, Inherited Velocity Speed Limit=0, Speed Limit=false, UseSpeedThreshold=false) {enabled=false}
  ]
  particleUpdateScript [
   &ParticleState()
   &ScaleSpriteSize(
    Scale Sprite Size Mode=::NewEnumerator0
    Uniform Scale Factor=&FloatFromCurve(FloatCurve=DataInterface() {propertyValues {Curve {keys [
         <at 0 1 RCIM_Cubic>
         {arriveTangent=2.8138, interpMode=RCIM_Cubic, leaveTangent=2.8138, time=0.1012, value=2.3784}
         {arriveTangent=-2.3275, interpMode=RCIM_Cubic, leaveTangent=-2.3275, time=0.4917, value=2.3837}
         {arriveTangent=-5.4419, interpMode=RCIM_Cubic, leaveTangent=-5.4419, time=0.7025, value=0.979}
         {arriveTangent=-3.2904, interpMode=RCIM_Cubic, leaveTangent=-3.2904, time=0.8788, value=0.2772}
         <at 1 0 RCIM_Cubic>
        ]}}})
   )
   &AccelerationForce(Acceleration=<0 0 200>)
   &Drag(Drag=0.5, Use Linear Drag=true, Use Rotational Drag=true)
   &SolveForcesAndVelocity(Clamp Velocity=false, Limit Acceleration=false)
   /Niagara/Modules/CascadeConversion/CascadeConversion_Orbit(
    Offset:Vector3f=<3 3 3>
    Rotation:Vector3f=&RandomRangeVector(Minimum=<0 0 0>, Maximum=<1 1 1>)
    Rotation Rate:Vector3f=&RandomRangeVector(Minimum=<0 0 0>, Maximum=<3 3 3>)
   )
   /Niagara/Modules/CascadeConversion/CascadeConversion_SolveOrbit()
   &ScaleColor(
    Scale Alpha=&FloatFromCurve(FloatCurve=DataInterface() {propertyValues {Curve {keys [
         {arriveTangent=-0.4877, time=0, value=1}
         {arriveTangent=-0.405, leaveTangent=-0.405, time=0.6336, value=0.5123}
         {arriveTangent=-1.3983, interpMode=RCIM_Cubic, leaveTangent=-1.3983, time=0.7427, value=0.1073}
         <at 1 0>
        ]}}})
    ScaleA=true
    ScaleRGB=true
   )
  ]
  rendererClasses [&NiagaraSpriteRendererProperties]
  renderers [
   {rendererClass=&NiagaraSpriteRendererProperties}
   {rendererClass=&NiagaraSpriteRendererProperties, rendererIndex=1}
  ]
 }
 Empty {
  bEnabled=false
  emitterUpdateScript [
   &EmitterState()
   &SpawnBurst_Instantaneous(Use Spawn Probability=false, UseLoopCountLimit=false)
  ]
  particleSpawnScript [
   &InitializeParticle(
    Color Mode=::NewEnumerator0
    Lifetime=3
    Sprite Size Mode=::NewEnumerator3
    Uniform Sprite Size=250
    UsePositionOffset=false
    Write Material Random=true
   )
   &CameraOffset(Camera Offset Amount=100)
  ]
  particleUpdateScript [
   &ParticleState()
   &DynamicMaterialParameters(
    Index 0 Param 1=$User.IsAlt
    Module.Index0Param0Write=true
    Module.Index0Param1Write=true
    Module.Index0Param2Write=true
    Module.Index0Param3Write=true
   )
  ]
  rendererClasses [&NiagaraSpriteRendererProperties]
  renderers [{rendererClass=&NiagaraSpriteRendererProperties}]
 }
}
summary {
 systemName=NS_JellyfishGlider_Bubble_Ejection
 emitters [
  {emitterName=Bubbles, rendererClasses [&NiagaraSpriteRendererProperties]}
  {bEnabled=false, emitterName=Empty, rendererClasses [&NiagaraSpriteRendererProperties]}
 ]
 userVariables [User.IsAlt {type=NiagaraFloat, defaultValue=0, description=""}]
}
